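; Game state component for the mid-match weapon bunker objective, choosing which bunkers exist and opening them, server-side.
; Once playlist data is ready, it shuffles hotfix-enabled DataTable rows and activates each row's InUse or PermanentlyClosed World Partition data layer, with hotfix force overrides.
; On objective activation, it marks the nearest storm-safe door for each unplaced player and starts every TidalCrane door's opening sequence.
; Debug cheats trigger activation, toggle locations and show minimap icons.
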(event EventBeginPlay
  ;[0.03427 0.1562 0.03703] BeginPlay: Based on the hotfix values determine the datalayers to activate and activate them.
  (|CallParentFunction)
  (Networking|SwitchHasAuthority
    (:Authority
      (bind _gamestate (Class|FortAsyncActionCurrentPlaylistReady|PlaylistDataReadyAsync)
        (:OnReady
          (Utilities|IsValid _gamestate
            (:Is%20Valid
              (if (AI|Navigation|IsNavigationBeingBuiltorLocked)
                (Utilities|FlowControl|Delay 0.5) ; Make sure NavMesh is not being Locked or Built so when we load the DataLayers, the NavMesh gets updated properly.
              )
              (CallFunction|CacheHotfixEnabledDataLayerRowEntries)
              (CallFunction|DetermineDataLayerRowEntriestoActivate)
              (CallFunction|FinalizeDataLayerRuntimeState))))))))

(event EventActivateObjective
  ;[0.0396 0.8177 0.2744] ActivateObjective: Mark the closest bunker door for all players and start the opening sequence on all of them.
  (|CallParentFunction)
  (Networking|SwitchHasAuthority
    (:Authority
      (bind _outactors (Actor|GetAllActorsOfClass "/MidMatchWeaponBunkerGameplay/Items/WeaponBunkerDoor/MMOBunkers_Ch6S4/BGA_BunkerDoor_TidalCrane.BGA_BunkerDoor_TidalCrane_C"))
      (Default|MarkClosestInuseBunkerDoorforAllPlayers (CallFunction|FilterOutBunkersinStorm _outactors))
      (for _array_element _outactors
        (Utilities|IsValid _array_element
          (:Is%20Valid
            (Class|BGAWeaponBunkerDoorParent|StartOpeningSequence _array_element)))))))

(event Custom|DebugWeaponBunkerActivateObjective
  ;[0 0 0] Cheat ke * DebugWeaponBunkerActivateObjective
  ;
  ; Activate the weapon bunker objective.
  (Networking|SwitchHasAuthority
    (:Authority
      (CallFunction|ActivateObjective))))

(event Custom|DebugWeaponBunkerEnableLocations
  (params (Input :type String))
  (bind _returnvalue (Utilities|String|ParseIntoArray (Utilities|String|ToLower Input) ","))
  (bind _returnvalue_2 (Utilities|String|ToLower Input))
  (bind _returnvalue_3 (Utilities|String|Equal,CaseInsensitive(String) _returnvalue_2 "all"))
  ;[0 0 0] Cheat ke * DebugWeaponBunkerEnableLocations (names)
  ;
  ; Enables the desired weapon bunkers, disables the rest.
  (Networking|SwitchHasAuthority
    (:Authority
      (bind _weaponbunkerdatalayerinfodatatable (Variables|Settings|GetWeaponBunkerDataLayerInfoDataTable)
        (:then
          (bind _outrownames (DataTable|GetDataTableRowNames _weaponbunkerdatalayerinfodatatable))
          (for _array_element _outrownames
            (bind _returnvalue_4 (Utilities|GetDataTableRow _weaponbunkerdatalayerinfodatatable _array_element))
            (Utilities|IsValid (.InUse_12_74D9D88748D09A74C28DB093303FF73D (Utilities|Struct|BreakMidMatchWeaponBunkerDataLayerInfo _returnvalue_4))
              (:Is%20Valid
                (Utilities|IsValid (.PermanentlyClosed_11_C8877273432712EE0E2AD8A5284717CA (Utilities|Struct|BreakMidMatchWeaponBunkerDataLayerInfo _returnvalue_4))
                  (:Is%20Valid
                    (bind _returnvalue_5 (WorldPartition|GetDataLayerManager))
                    (Utilities|IsValid _returnvalue_5
                      (:Is%20Valid
                        (DataLayers|SetDataLayerRuntimeState _returnvalue_5 (.InUse_12_74D9D88748D09A74C28DB093303FF73D (Utilities|Struct|BreakMidMatchWeaponBunkerDataLayerInfo _returnvalue_4)) (select (or _returnvalue_3 (Utilities|Array|ContainsItem _returnvalue (Utilities|String|ToLower (Utilities|String|ToString(Name) _array_element)))) "Activated" "Unloaded") false :!class /Script/Engine.DataLayerManager) ; InUse
                        (DataLayers|SetDataLayerRuntimeState _returnvalue_5 (.PermanentlyClosed_11_C8877273432712EE0E2AD8A5284717CA (Utilities|Struct|BreakMidMatchWeaponBunkerDataLayerInfo _returnvalue_4)) (select (or _returnvalue_3 (Utilities|Array|ContainsItem _returnvalue (Utilities|String|ToLower (Utilities|String|ToString(Name) _array_element)))) "Unloaded" "Activated") false :!class /Script/Engine.DataLayerManager) ; PermanentlyClosed
                      ))))))))
        (:else)))))

(event Custom|DebugWeaponBunkerDisableLocations
  (params (Input :type String))
  (bind _returnvalue (Utilities|String|Equal,CaseInsensitive(String) (Utilities|String|ToLower Input) "all"))
  (bind _returnvalue_2 (Utilities|String|ParseIntoArray (Utilities|String|ToLower Input) ","))
  (bind _returnvalue_3 (Utilities|String|ToLower Input))
  ;[0 0 0] Cheat ke * DebugWeaponBunkerDisableLocations (names)
  ;
  ; Disable the desired weapon bunkers, enables the rest.
  (Networking|SwitchHasAuthority
    (:Authority
      (bind _weaponbunkerdatalayerinfodatatable (Variables|Settings|GetWeaponBunkerDataLayerInfoDataTable)
        (:then
          (bind _outrownames (DataTable|GetDataTableRowNames _weaponbunkerdatalayerinfodatatable))
          (for _array_element _outrownames
            (bind _returnvalue_4 (Utilities|GetDataTableRow _weaponbunkerdatalayerinfodatatable _array_element))
            (Utilities|IsValid (.InUse_12_74D9D88748D09A74C28DB093303FF73D (Utilities|Struct|BreakMidMatchWeaponBunkerDataLayerInfo _returnvalue_4))
              (:Is%20Valid
                (Utilities|IsValid (.PermanentlyClosed_11_C8877273432712EE0E2AD8A5284717CA (Utilities|Struct|BreakMidMatchWeaponBunkerDataLayerInfo _returnvalue_4))
                  (:Is%20Valid
                    (bind _returnvalue_5 (WorldPartition|GetDataLayerManager))
                    (Utilities|IsValid _returnvalue_5
                      (:Is%20Valid
                        (DataLayers|SetDataLayerRuntimeState _returnvalue_5 (.InUse_12_74D9D88748D09A74C28DB093303FF73D (Utilities|Struct|BreakMidMatchWeaponBunkerDataLayerInfo _returnvalue_4)) (select (or _returnvalue (Utilities|Array|ContainsItem _returnvalue_2 (Utilities|String|ToLower (Utilities|String|ToString(Name) _array_element)))) "Unloaded" "Activated") false :!class /Script/Engine.DataLayerManager) ; InUse
                        (DataLayers|SetDataLayerRuntimeState _returnvalue_5 (.PermanentlyClosed_11_C8877273432712EE0E2AD8A5284717CA (Utilities|Struct|BreakMidMatchWeaponBunkerDataLayerInfo _returnvalue_4)) (select (or _returnvalue (Utilities|Array|ContainsItem _returnvalue_2 (Utilities|String|ToLower (Utilities|String|ToString(Name) _array_element)))) "Activated" "Unloaded") false :!class /Script/Engine.DataLayerManager) ; PermanentlyClosed
                      ))))))))
        (:else)))))

(event Custom|DebugWeaponBunkerShowOnMap
  (params (bShow :type Boolean))
  ;[0 0 0] Cheat ke * DebugWeaponBunkerShowOnMap (1/0)
  ;
  ; Show the weapon bunkers in the world on the minimap.
  (Networking|SwitchHasAuthority
    (:Authority
      (bind _outactors (Actor|GetAllActorsOfClass "/MidMatchWeaponBunkerGameplay/Debug/B_WeaponBunkerDebugMinimapIconActorParent.B_WeaponBunkerDebugMinimapIconActorParent_C"))
      (Actor|DestroyActor _outactors)
      (if bShow
        (bind _outactors_2 (Actor|GetAllActorsOfClass "/Script/FortniteGame.FortStaticMeshActor"))
        (for _array_element _outactors_2
          (Utilities|IsValid _array_element
            (:Is%20Valid
              (if (Actor|ActorHasTag _array_element "WeaponBunkerSealed")
                (Game|SpawnActorfromClass "/MidMatchWeaponBunkerGameplay/Debug/B_WeaponBunkerDebugMinimapIconActor_Sealed.B_WeaponBunkerDebugMinimapIconActor_Sealed_C" (Transformation|GetActorTransform _array_element) "AlwaysSpawn" "MultiplyWithRoot" 0 0)))))
        (bind _outactors_3 (Actor|GetAllActorsOfClass "/MidMatchWeaponBunkerGameplay/Items/WeaponBunkerDoor/BGA_WeaponBunkerDoor.BGA_WeaponBunkerDoor_C"))
        (for _array_element_2 _outactors_3
          (Utilities|IsValid _array_element_2
            (:Is%20Valid
              (Game|SpawnActorfromClass "/MidMatchWeaponBunkerGameplay/Debug/B_WeaponBunkerDebugMinimapIconActor_Unsealed.B_WeaponBunkerDebugMinimapIconActor_Unsealed_C" (Transformation|GetActorTransform _array_element_2) "AlwaysSpawn" "MultiplyWithRoot" 0 0))))))))

(fn CacheHotfixEnabledDataLayerRowEntries
  ;[0.03001 0.1146 0.102] CacheHotfixEnabledDataLayerRowEntries: Cache the row names for the various data layer row entries that are hotfixed on.
  (Networking|SwitchHasAuthority
    (:Authority
      (bind _weaponbunkerdatalayerinfodatatable (Variables|Settings|GetWeaponBunkerDataLayerInfoDataTable)
        (:then
          (bind _outrownames (DataTable|GetDataTableRowNames _weaponbunkerdatalayerinfodatatable))
          (for _array_element _outrownames
            (bind _returnvalue (Utilities|GetDataTableRow _weaponbunkerdatalayerinfodatatable _array_element))
            (if (ScalableFloat|GetValueasBool (Utilities|Struct|BreakMidMatchWeaponBunkerDataLayerInfo _returnvalue))
              (Utilities|IsValid (.InUse_12_74D9D88748D09A74C28DB093303FF73D (Utilities|Struct|BreakMidMatchWeaponBunkerDataLayerInfo _returnvalue))
                (:Is%20Valid
                  (Utilities|IsValid (.PermanentlyClosed_11_C8877273432712EE0E2AD8A5284717CA (Utilities|Struct|BreakMidMatchWeaponBunkerDataLayerInfo _returnvalue))
                    (:Is%20Valid
                      (Utilities|Array|AddUnique (Variables|Private|GetHotfixEnabledDataLayerRowEntries) _array_element)
                      (Fortnite|FortShippingLog self (Utilities|String|Append "CacheHotfixEnabledDataLayerRowEntries: Enabled row: " (Utilities|String|ToString(Name) _array_element))))
                    (:Is%20Not%20Valid
                      (Fortnite|FortShippingLog self (Utilities|String|Append "CacheHotfixEnabledDataLayerRowEntries: PermanentlyClosed datalayer for row is null: ") true))))
                (:Is%20Not%20Valid
                  (Fortnite|FortShippingLog self (Utilities|String|Append "CacheHotfixEnabledDataLayerRowEntries: InUse datalayer for row is null: " (Utilities|String|ToString(Name) _array_element)) true))))))
        (:else
          (Fortnite|FortShippingLog self "CacheHotfixEnabledDataLayerRowEntries: WeaponBunkerDataLayerInfoDataTable is null." true))))))

(fn DetermineDataLayerRowEntriesToActivate
  (bind _hotfix_numofactivebunkers (Variables|Hotfix|GetHotfix_NumOfActiveBunkers))
  (bind _returnvalue (ScalableFloat|GetValueasInteger _hotfix_numofactivebunkers))
  (bind _hotfixenableddatalayerrowentries (Variables|Private|GetHotfixEnabledDataLayerRowEntries))
  ;[0.2708 0.1811 0.1055] DetermineDataLayerRowEntriesToActivate: Shuffle the hotfix enabled datalayer row entries and cache the desired number of them as the entries to activate.
  (Networking|SwitchHasAuthority
    (:Authority
      (Utilities|Array|Shuffle (Variables|Private|GetHotfixEnabledDataLayerRowEntries))
      (bind _index (Utilities|FlowControl|ForLoopwithBreak 0 (- _returnvalue 1))
        (:LoopBody
          (if (not (Utilities|Array|IsValidIndex (Variables|Private|GetHotfixEnabledDataLayerRowEntries) _index))
            (Fortnite|FortShippingLog self (Utilities|String|Append "DetermineDataLayerRowEntriesToActivate: Not enough number of data layer row entries are available to reach the desired count of " (Utilities|String|ToString(Integer) _returnvalue)) true))
          (Utilities|Array|AddUnique (Variables|Private|GetDataLayerRowEntriesToActivate) (Utilities|Array|Get(aref) _hotfixenableddatalayerrowentries _index))
          (Fortnite|FortShippingLog self (Utilities|String|Append "DetermineDataLayerRowEntriesToActivate: Setting datalayer row to be activated: " (Utilities|String|ToString(Name) (Utilities|Array|Get(aref) _hotfixenableddatalayerrowentries _index)))))))))

(fn FilterOutBunkersInStorm
  (params (BunkerDoors :type Array%20of%20BGA%20Weapon%20Bunker%20Door%20Parent%20Object%20References))
  (bind _returnvalue (SafeZone|GetSafeZoneIndicator))
  ;[0.007339 0.2061 0.2396] FilterOutBunkersInStorm: Returns an array of weapon bunkers that are not in the storm.
  (Networking|SwitchHasAuthority
    (:Authority
      (Utilities|FlowControl|Sequence
        (:then_0
          (if (>= (ScalableFloat|GetValueatLevel (Variables|Hotfix|GetHotfix_MinDistanceFromStormForMarking)) 0)
            (Utilities|IsValid _returnvalue
              (:Is%20Valid
                (for _array_element BunkerDoors
                  (Utilities|IsValid _array_element
                    (:Is%20Valid
                      (if (< (Math|Vector|Distance2DSquared(Vector) (+ (Transformation|GetActorLocation _array_element) (* (Math|Vector|Normalize2D(Vector) (Math|Vector|GetUnitDirection(Vector) (SafeZone|GetSafeZoneCenter _returnvalue) (Transformation|GetActorLocation _array_element))) (ScalableFloat|GetValueatLevel (Variables|Hotfix|GetHotfix_MinDistanceFromStormForMarking)))) (SafeZone|GetSafeZoneCenter _returnvalue)) (Math|Float|Square (SafeZone|GetSafeZoneRadius _returnvalue)))
                        (Utilities|Array|Add (LocalVariables|GetReturnBunkers) _array_element))))))
              (:Is%20Not%20Valid
                _))
            (else
              _))
          (return BunkerDoors))
        (:then_1
          (return (LocalVariables|GetReturnBunkers)))))))

(fn FinalizeDataLayerRuntimeState
  (bind _hotfix_forceactivateallbunkerstoclosedstate (Variables|Hotfix|GetHotfix_ForceActivateAllBunkersToClosedState))
  (bind _returnvalue (ScalableFloat|GetValueasBool _hotfix_forceactivateallbunkerstoclosedstate))
  (bind _hotfix_forceactivateallbunkerstoopenstate (Variables|Hotfix|GetHotfix_ForceActivateAllBunkersToOpenState))
  (bind _returnvalue_2 (ScalableFloat|GetValueasBool _hotfix_forceactivateallbunkerstoopenstate))
  (bind _benabled (CallFunction|IsObjectiveEnabled))
  (bind _datalayerrowentriestoactivate (Variables|Private|GetDataLayerRowEntriesToActivate))
  ;[0.01069 0.1391 0.1719] FinalizeDataLayerRuntimeState: Activate the decided upon datalayers.
  (Networking|SwitchHasAuthority
    (:Authority
      (bind _returnvalue_3 (WorldPartition|GetDataLayerManager))
      (Utilities|IsValid _returnvalue_3
        (:Is%20Valid
          (bind _weaponbunkerdatalayerinfodatatable (Variables|Settings|GetWeaponBunkerDataLayerInfoDataTable)
            (:then
              (bind _outrownames (DataTable|GetDataTableRowNames _weaponbunkerdatalayerinfodatatable))
              (for _array_element _outrownames
                (bind _returnvalue_4 (Utilities|GetDataTableRow _weaponbunkerdatalayerinfodatatable _array_element))
                (bind _returnvalue_5 (WorldPartition|GetDataLayerManager))
                (bind _returnvalue_6 (DataLayers|SetDataLayerRuntimeState _returnvalue_5 (select _returnvalue_2 (.InUse_12_74D9D88748D09A74C28DB093303FF73D (Utilities|Struct|BreakMidMatchWeaponBunkerDataLayerInfo _returnvalue_4)) (select _returnvalue (.PermanentlyClosed_11_C8877273432712EE0E2AD8A5284717CA (Utilities|Struct|BreakMidMatchWeaponBunkerDataLayerInfo _returnvalue_4)) (select (and (Utilities|Array|ContainsItem _datalayerrowentriestoactivate _array_element) _benabled) (.InUse_12_74D9D88748D09A74C28DB093303FF73D (Utilities|Struct|BreakMidMatchWeaponBunkerDataLayerInfo _returnvalue_4)) (.PermanentlyClosed_11_C8877273432712EE0E2AD8A5284717CA (Utilities|Struct|BreakMidMatchWeaponBunkerDataLayerInfo _returnvalue_4))))) "Activated" false :!class /Script/Engine.DataLayerManager))
                (Fortnite|FortShippingLog self (select _returnvalue_6 (Utilities|String|Append "FinalizeDataLayerRuntimeState: Successfully set datalayer as Activated: " (Utilities|GetDisplayName (select _returnvalue_2 (.InUse_12_74D9D88748D09A74C28DB093303FF73D (Utilities|Struct|BreakMidMatchWeaponBunkerDataLayerInfo _returnvalue_4)) (select _returnvalue (.PermanentlyClosed_11_C8877273432712EE0E2AD8A5284717CA (Utilities|Struct|BreakMidMatchWeaponBunkerDataLayerInfo _returnvalue_4)) (select (and (Utilities|Array|ContainsItem _datalayerrowentriestoactivate _array_element) _benabled) (.InUse_12_74D9D88748D09A74C28DB093303FF73D (Utilities|Struct|BreakMidMatchWeaponBunkerDataLayerInfo _returnvalue_4)) (.PermanentlyClosed_11_C8877273432712EE0E2AD8A5284717CA (Utilities|Struct|BreakMidMatchWeaponBunkerDataLayerInfo _returnvalue_4))))))) (Utilities|String|Append "FinalizeDataLayerRuntimeState: Failed to set datalayer as Activated: " (Utilities|GetDisplayName (select _returnvalue_2 (.InUse_12_74D9D88748D09A74C28DB093303FF73D (Utilities|Struct|BreakMidMatchWeaponBunkerDataLayerInfo _returnvalue_4)) (select _returnvalue (.PermanentlyClosed_11_C8877273432712EE0E2AD8A5284717CA (Utilities|Struct|BreakMidMatchWeaponBunkerDataLayerInfo _returnvalue_4)) (select (and (Utilities|Array|ContainsItem _datalayerrowentriestoactivate _array_element) _benabled) (.InUse_12_74D9D88748D09A74C28DB093303FF73D (Utilities|Struct|BreakMidMatchWeaponBunkerDataLayerInfo _returnvalue_4)) (.PermanentlyClosed_11_C8877273432712EE0E2AD8A5284717CA (Utilities|Struct|BreakMidMatchWeaponBunkerDataLayerInfo _returnvalue_4)))))))) (not _returnvalue_6))
                (if (ScalableFloat|GetValueasBool (Variables|Hotfix|GetHotfix_ForceActivateAllBunkersToOpenState))
                  (Fortnite|FortShippingLog self "FinalizeDataLayerRuntimeState: Force Actiavte Bunkers To Open State Is Enabled - This should only happen in development builds." true)
                  (elif (ScalableFloat|GetValueasBool (Variables|Hotfix|GetHotfix_ForceActivateAllBunkersToClosedState))
                    (Fortnite|FortShippingLog self "FinalizeDataLayerRuntimeState: Force Actiavte Bunkers To Closed State Is Enabled - This should only happen in development builds." true)))))
            (:else)))
        (:Is%20Not%20Valid
          (Fortnite|FortShippingLog self "FinalizeDataLayerRuntimeState: DataLayerManager is null." true))))))

(fn MarkClosestInuseBunkerDoorForAllPlayers
  (params (BunkerDoors :type Array%20of%20Actor%20Object%20References))
  ;[0.01737 0.1198 0.08191] MarkClosestActiveBunkerDoorForAllPlayers: Mark the closest active bunker doors for all players.
  (Networking|SwitchHasAuthority
    (:Authority
      (for _array_element (Fortnite|GetAllFortPlayerControllers true true)
        (Utilities|IsValid _array_element
          (:Is%20Valid
            (Utilities|IsValid (Pawn|GetViewTarget _array_element)
              (:Is%20Valid
                (bind _asfort_player_state_athena (Utilities|Casting|CastToFortPlayerStateAthena (Variables|/Script/Engine.Controller|GetPlayerState _array_element)))
                (if (<= (Variables|/Script/FortniteGame.FortPlayerStateAthena|GetPlace _asfort_player_state_athena) 0)
                  (Utilities|IsValid (Actor|GetComponentbyClass _array_element "/Script/FortniteGame.AthenaMarkerComponent")
                    (:Is%20Valid
                      (bind _asbga_weapon_bunker_door_parent (Utilities|Casting|CastToBGA_WeaponBunkerDoor_Parent (Fortnite|GetClosestActorfromArray (Pawn|GetViewTarget _array_element) BunkerDoors)))
                      (MarkerComponent|MarkActoronServer (Actor|GetComponentbyClass _array_element "/Script/FortniteGame.AthenaMarkerComponent") _asbga_weapon_bunker_door_parent false true))))))))))))
