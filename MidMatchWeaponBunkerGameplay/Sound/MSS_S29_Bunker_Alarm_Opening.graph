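# Stereo one-shot MetaSound source that plays the bunker door opening alert alarm.
# On play, a short delay starts a trigger repeat that retriggers the wave player a set number of times at a fixed period.
# Attenuation distance drives both per-channel low-pass cutoffs, darkening the sound farther away, and a distance-mapper patch that scales volume.
# The wave player's On Finished feeds the source's OnFinished directly, independent of the repeat count.
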MetaSoundSource /MidMatchWeaponBunkerGameplay/Sound/MSS_S29_Bunker_Alarm_Opening
inputs [
 Alarm Sound {type=WaveAsset}
 UE.Attenuation.Distance {type=Float}
 UE.Source.OnPlay {type=Trigger}
]
interfaces [UE.Attenuation:1.0, UE.OutputFormat.Stereo:1.0, UE.Source.OneShot:1.0, UE.Source:1.1]
outputs [
 "UE.OutputFormat.Stereo.Audio:0" {type=Audio}
 "UE.OutputFormat.Stereo.Audio:1" {type=Audio}
 UE.Source.OneShot.OnFinished {type=Trigger}
]
pages [Default {
  input_defaults {Alarm Sound=/MidMatchWeaponBunkerGameplay/Sound/wav/S29_MMO_Bunker_Door_Alert_Opening, UE.Attenuation.Distance=0, UE.Source.OnPlay=null}
  nodes [
   n1 = /FortMetasound/Patches/MSP_DistanceMapper(Distance=^UE.Attenuation.Distance) {also {defaults {Falloff=8500, Inner Radius=1000, Inner Volume=1.2}}}
   n2 = MapRange.MapRange:Float(In=^UE.Attenuation.Distance) {also {defaults {In Range B=2000, Out Range A=20000, Out Range B=1000}}}
   TriggerDelay_0 = UE.Trigger Delay(In=^UE.Source.OnPlay) {also {defaults {Delay Time=0.4}}}
   WavePlayer_0 = UE.Wave Player:Stereo(Wave Asset=^Alarm Sound, Play=UE.TriggerRepeat(Start=@TriggerDelay_0.Out) {also {defaults {Num Repeats=3, Period=2.5}}}.RepeatOut)
   OnePoleLowPassFilter_0 = UE.One-Pole Low Pass Filter:Audio(Cutoff Frequency=@n2.Out Value, In=@WavePlayer_0.Out Right)
   Multiply_0 = UE.Multiply:Audio by Float(AdditionalOperands=@n1.Output, PrimaryOperand=@OnePoleLowPassFilter_0.Out)
   OnePoleLowPassFilter_1 = UE.One-Pole Low Pass Filter:Audio(Cutoff Frequency=@n2.Out Value, In=@WavePlayer_0.Out Left)
   Multiply_1 = UE.Multiply:Audio by Float(AdditionalOperands=@n1.Output, PrimaryOperand=@OnePoleLowPassFilter_1.Out)
  ]
  links [
   {dst="@UE.OutputFormat.Stereo.Audio:1", src=Multiply_0.Out}
   {dst="@UE.OutputFormat.Stereo.Audio:0", src=Multiply_1.Out}
   {dst="@UE.Source.OneShot.OnFinished", src=WavePlayer_0.On Finished}
  ]
 }]
properties {
 assetImportData=&self:AssetImportData
 attenuationSettings=&/Att_S29_Bunker
 maxDistance=34200
 numChannels=2
 outputFormat=Stereo
 sampleRate=48000
 soundClassObject=&Fort_Ambient_Props_Athena
 volume=0.501
}
