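# Mono MetaSound source for the looping ambient alarm at the mid-match weapon bunker.
# On play, a single Wave Player feeds the alarm loop wave straight to the output.
# It uses the bunker smoke attenuation and the doors sound class.
# The Distance input and the three door wave inputs (open start, moving, stop) are declared but not connected.
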MetaSoundSource /MidMatchWeaponBunkerGameplay/Sound/MSS_S29_Bunker_Amb_Alarm_loop
inputs [
 Distance {type=Float}
 Door Moving Sound {type=WaveAsset}
 Door Open Start {type=WaveAsset}
 Door Stop Sound {type=WaveAsset}
 UE.Source.OnPlay {type=Trigger}
]
interfaces [UE.OutputFormat.Mono:1.0, UE.Source:1.1]
outputs ["UE.OutputFormat.Mono.Audio:0" {type=Audio}]
pages [Default {
  input_defaults {
   Distance=0
   Door Moving Sound=/MidMatchWeaponBunkerGameplay/Sound/wav/S29_MMO_Bunker_Door_Moving
   Door Open Start=/MidMatchWeaponBunkerGameplay/Sound/wav/S29_MMO_Bunker_Door_Open_Start
   Door Stop Sound=/MidMatchWeaponBunkerGameplay/Sound/wav/S29_MMO_Bunker_Door_Open_End
   UE.Source.OnPlay=null
  }
  nodes [WavePlayer_0 = UE.Wave Player:Mono(Play=^UE.Source.OnPlay) {also {defaults {Loop="<MetaSoundBuilderResult.SUCCEEDED: 0>", Wave Asset=/MidMatchWeaponBunkerGameplay/Sound/wav/S29_MMO_Bunker_Amb_Alarm_lp}}}]
  links [{dst="@UE.OutputFormat.Mono.Audio:0", src=WavePlayer_0.Out Mono}]
 }]
properties {
 assetImportData=&self:AssetImportData
 attenuationSettings=&/Att_S29_Bunker_Smoke
 bLooping=true
 duration=10000
 maxDistance=2800
 sampleRate=48000
 soundClassObject=&Fort_Doors
 volume=1.4
}
