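# Stereo one-shot MetaSound for the bunker door opening, layering start, moving and stop sounds.
# The start wave plays immediately, while two trigger delays stagger the moving loop and the later stop sound into a shared mixer.
# Attenuation distance drives both per-channel low-pass cutoff and MSP_DistanceMapper gain, so the door gets duller and quieter far away.
# OnFinished fires once all three wave players have finished.
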MetaSoundSource /MidMatchWeaponBunkerGameplay/Sound/MSS_S29_Bunker_Door_Open
inputs [
 Distance {type=Float}
 Door Moving Sound {type=WaveAsset}
 Door Open Start {type=WaveAsset}
 Door Stop Sound {type=WaveAsset}
 UE.Attenuation.Distance {type=Float}
 UE.Source.OnPlay {type=Trigger}
]
interfaces [UE.Attenuation:1.0, UE.OutputFormat.Stereo:1.0, UE.Source.OneShot:1.0, UE.Source:1.1]
outputs [
 "UE.OutputFormat.Stereo.Audio:0" {type=Audio}
 "UE.OutputFormat.Stereo.Audio:1" {type=Audio}
 UE.Source.OneShot.OnFinished {type=Trigger}
]
pages [Default {
  input_defaults {
   Distance=0
   Door Moving Sound=/MidMatchWeaponBunkerGameplay/Sound/wav/S29_MMO_Bunker_Door_Moving
   Door Open Start=/MidMatchWeaponBunkerGameplay/Sound/wav/S29_MMO_Bunker_Door_Open_Start
   Door Stop Sound=/MidMatchWeaponBunkerGameplay/Sound/wav/S29_MMO_Bunker_Door_Open_End
   UE.Attenuation.Distance=0
   UE.Source.OnPlay=null
  }
  nodes [
   n1 = /FortMetasound/Patches/MSP_DistanceMapper(Distance=^UE.Attenuation.Distance) {also {defaults {Falloff=29000, Inner Radius=1200, Inner Volume=1}}}
   n2 = MapRange.MapRange:Float(In=^UE.Attenuation.Distance) {also {defaults {In Range B=5000, Out Range A=8000, Out Range B=600}}}
   TriggerDelay_1 = UE.Trigger Delay(In=^UE.Source.OnPlay) {also {defaults {Delay Time=0.5}}}
   WavePlayer_0 = UE.Wave Player:Stereo(Wave Asset=^Door Moving Sound, Play=@TriggerDelay_1.Out)
   TriggerDelay_0 = UE.Trigger Delay(In=^UE.Source.OnPlay) {also {defaults {Delay Time=7.5}}}
   WavePlayer_1 = UE.Wave Player:Stereo(Wave Asset=^Door Stop Sound, Play=@TriggerDelay_0.Out)
   WavePlayer_2 = UE.Wave Player:Stereo(Wave Asset=^Door Open Start, Play=^UE.Source.OnPlay)
   AudioMixerAudioMixerStereo3_0 = "AudioMixer.Audio Mixer (Stereo, 3)"(
    In 1 L=@WavePlayer_0.Out Left
    In 1 R=@WavePlayer_0.Out Right
    In 2 L=@WavePlayer_1.Out Left
    In 2 R=@WavePlayer_1.Out Right
    In 0 L=@WavePlayer_2.Out Left
    In 0 R=@WavePlayer_2.Out Right
   )
   OnePoleLowPassFilter_1 = UE.One-Pole Low Pass Filter:Audio(Cutoff Frequency=@n2.Out Value, In=UE.Reroute(Value=UE.Reroute(Value=@AudioMixerAudioMixerStereo3_0.Out R).Value).Value)
   Multiply_0 = UE.Multiply:Audio by Float(AdditionalOperands=@n1.Output, PrimaryOperand=@OnePoleLowPassFilter_1.Out)
   OnePoleLowPassFilter_0 = UE.One-Pole Low Pass Filter:Audio(Cutoff Frequency=@n2.Out Value, In=UE.Reroute(Value=UE.Reroute(Value=@AudioMixerAudioMixerStereo3_0.Out L).Value).Value)
   Multiply_1 = UE.Multiply:Audio by Float(AdditionalOperands=@n1.Output, PrimaryOperand=@OnePoleLowPassFilter_0.Out)
   TriggerAccumulatorTriggerAccumulate3_0 = "TriggerAccumulator.Trigger Accumulate (3)"(In 1=@WavePlayer_0.On Finished, In 2=@WavePlayer_1.On Finished, In 0=@WavePlayer_2.On Finished)
  ]
  links [
   {dst="@UE.OutputFormat.Stereo.Audio:1", src=Multiply_0.Out}
   {dst="@UE.OutputFormat.Stereo.Audio:0", src=Multiply_1.Out}
   {dst="@UE.Source.OneShot.OnFinished", src=TriggerAccumulatorTriggerAccumulate3_0.Out}
  ]
 }]
properties {
 assetImportData=&self:AssetImportData
 attenuationSettings=&/Att_S29_Bunker
 maxDistance=34200
 numChannels=2
 outputFormat=Stereo
 sampleRate=48000
 soundClassObject=&Fort_Doors
 volume=0.5
}
