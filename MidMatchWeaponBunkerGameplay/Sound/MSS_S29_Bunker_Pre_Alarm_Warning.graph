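# One-shot stereo MetaSound source that plays a bunker door blast warning alert, used as the pre-alarm warning in the mid-match weapon bunker gameplay.
# A single Wave Player starts on play, feeds both output channels and signals finished when the wave ends.
# It uses underground bunker attenuation and the ambient props sound class.
# A concurrency override allows only one instance at a time, preventing new ones and setting a retrigger time.
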MetaSoundSource /MidMatchWeaponBunkerGameplay/Sound/MSS_S29_Bunker_Pre_Alarm_Warning
inputs [
 Pre Alarm Warning {type=WaveAsset}
 UE.Source.OnPlay {type=Trigger}
]
interfaces [UE.OutputFormat.Stereo:1.0, UE.Source.OneShot:1.0, UE.Source:1.1]
outputs [
 "UE.OutputFormat.Stereo.Audio:0" {type=Audio}
 "UE.OutputFormat.Stereo.Audio:1" {type=Audio}
 UE.Source.OneShot.OnFinished {type=Trigger}
]
pages [Default {
  input_defaults {Pre Alarm Warning=/MidMatchWeaponBunkerGameplay/Sound/wav/S29_MMO_Bunker_Door_Alert_BlastWarning, UE.Source.OnPlay=null}
  nodes [WavePlayer_0 = UE.Wave Player:Stereo(Wave Asset=^Pre Alarm Warning, Play=^UE.Source.OnPlay)]
  links [
   {dst="@UE.Source.OneShot.OnFinished", src=WavePlayer_0.On Finished}
   {dst="@UE.OutputFormat.Stereo.Audio:0", src=WavePlayer_0.Out Left}
   {dst="@UE.OutputFormat.Stereo.Audio:1", src=WavePlayer_0.Out Right}
  ]
 }]
properties {
 assetImportData=&self:AssetImportData
 attenuationSettings=&/Att_S29_Bunker_Underground
 bOverrideConcurrency=true
 maxDistance=7000
 numChannels=2
 outputFormat=Stereo
 sampleRate=48000
 soundClassObject=&Fort_Ambient_Props_Athena
 volume=1.504
 concurrencyOverrides {maxCount=1, resolutionRule=PreventNew, retriggerTime=5}
}
