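# One-shot stereo MetaSound for the weapon bunker's priming alarm, scaled by listener distance.
# The alarm plays once unfiltered while a second player retriggers it periodically through distance-driven low-pass filters; both are mixed, favoring the filtered layer.
# Delays derived from Priming Duration feed an accumulator that fires OnFinished.
# The long alarm, pre-opening tell and unlocking wave inputs are unconnected.
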MetaSoundSource /MidMatchWeaponBunkerGameplay/Sound/MSS_S29_Bunker_Priming
inputs [
 Alarm Sound {type=WaveAsset}
 Alarm Sound Long {type=WaveAsset}
 Pre Opening Tell {type=WaveAsset}
 Priming Duration {type=Float}
 UE.Attenuation.Distance {type=Float}
 UE.Source.OnPlay {type=Trigger}
 Underground Unlocking Sound {type=WaveAsset}
]
interfaces [UE.Attenuation:1.0, UE.OutputFormat.Stereo:1.0, UE.Source.OneShot:1.0, UE.Source:1.1]
outputs [
 "UE.OutputFormat.Stereo.Audio:0" {type=Audio}
 "UE.OutputFormat.Stereo.Audio:1" {type=Audio}
 UE.Source.OneShot.OnFinished {type=Trigger}
]
pages [Default {
  input_defaults {
   Alarm Sound=/MidMatchWeaponBunkerGameplay/Sound/wav/S29_MMO_Bunker_Door_Alert_Priming_long
   Alarm Sound Long=&S29_MMO_Bunker_Door_Alert_Priming_long
   Pre Opening Tell=/MidMatchWeaponBunkerGameplay/Sound/wav/S29_MMO_Bunker_Door_PreOpening_Open_Tell
   Priming Duration=30
   UE.Attenuation.Distance=32760.0312
   UE.Source.OnPlay=null
   Underground Unlocking Sound=/MidMatchWeaponBunkerGameplay/Sound/wav/S29_MMO_Bunker_Door_Unlocking
  }
  nodes [
   n1 = /FortMetasound/Patches/MSP_DistanceMapper(Distance=^UE.Attenuation.Distance) {also {defaults {Falloff=35000, Inner Volume=1}}}
   WavePlayer_0 = UE.Wave Player:Stereo(Wave Asset=^Alarm Sound, Play=^UE.Source.OnPlay)
   n2 = UE.Reroute(Value=@WavePlayer_0.Out Left)
   n3 = UE.Reroute(Value=@WavePlayer_0.Out Right)
   n4 = MapRange.MapRange:Float(In=^UE.Attenuation.Distance) {also {defaults {In Range B=8000, Out Range A=20000, Out Range B=900}}}
   TriggerAnyTriggerAny2_0 = "TriggerAny.Trigger Any (2)"(
    In 0=UE.Reroute(Value=@WavePlayer_0.On Play).Value
    In 1=UE.TriggerRepeat(Start=@WavePlayer_0.On Play) {also {defaults {Num Repeats=9, Period=3}}}.RepeatOut
   )
   WavePlayer_1 = UE.Wave Player:Stereo(Wave Asset=^Alarm Sound, Play=@TriggerAnyTriggerAny2_0.Out)
   OnePoleLowPassFilter_0 = UE.One-Pole Low Pass Filter:Audio(Cutoff Frequency=@n4.Out Value, In=@WavePlayer_1.Out Right)
   OnePoleLowPassFilter_1 = UE.One-Pole Low Pass Filter:Audio(Cutoff Frequency=@n4.Out Value, In=@WavePlayer_1.Out Left)
   AudioMixerAudioMixerStereo2_0 = "AudioMixer.Audio Mixer (Stereo, 2)"(
    In 1 L=UE.Multiply:Audio by Float(AdditionalOperands=@n1.Output, PrimaryOperand=UE.Reroute(Value=@n2.Value).Value).Out
    In 1 R=UE.Multiply:Audio by Float(AdditionalOperands=@n1.Output, PrimaryOperand=UE.Reroute(Value=@n3.Value).Value).Out
    In 0 R=UE.Multiply:Audio by Float(AdditionalOperands=@n1.Output, PrimaryOperand=@OnePoleLowPassFilter_0.Out).Out
    In 0 L=UE.Multiply:Audio by Float(AdditionalOperands=@n1.Output, PrimaryOperand=@OnePoleLowPassFilter_1.Out).Out
   ) {also {defaults {Gain 0=2}}}
   n5 = UE.Reroute(Value=^UE.Source.OnPlay)
   n6 = UE.ConversionFloatToTime(In=^Priming Duration)
   TriggerDelay_0 = UE.Trigger Delay(In=@n5.Value, Delay Time=UE.Subtract:Time(PrimaryOperand=@n6.Out) {also {defaults {AdditionalOperands=1.5}, version [1,1]}}.Out)
   TriggerDelay_1 = UE.Trigger Delay(Delay Time=@n6.Out, In=@n5.Value)
   TriggerAccumulatorTriggerAccumulate2_0 = "TriggerAccumulator.Trigger Accumulate (2)"(In 1=@TriggerDelay_0.Out, In 0=@TriggerDelay_1.Out)
  ]
  links [
   {dst="@UE.OutputFormat.Stereo.Audio:0", src=AudioMixerAudioMixerStereo2_0.Out L}
   {dst="@UE.OutputFormat.Stereo.Audio:1", src=AudioMixerAudioMixerStereo2_0.Out R}
   {dst="@UE.Source.OneShot.OnFinished", src=TriggerAccumulatorTriggerAccumulate2_0.Out}
  ]
 }]
properties {
 assetImportData=&self:AssetImportData
 attenuationSettings=&/Att_S29_Bunker
 maxDistance=34200
 numChannels=2
 outputFormat=Stereo
 sampleRate=48000
 soundClassObject=&Fort_Ambient_Props_Athena
 volume=0.451
}
