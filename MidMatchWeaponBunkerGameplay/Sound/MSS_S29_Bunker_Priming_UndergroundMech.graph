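# One-shot stereo MetaSound for the underground mechanism of a bunker while it primes, layering a door unlocking sound with a pre-opening tell.
# The unlocking sound triggers from OnPlay, while the tell is delayed until five seconds before Priming Duration ends.
# Listener distance scales both layers' volume and lowers the low-pass cutoff on the unlocking layer only.
# The layers mix to stereo, and the source finishes once both have ended.
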MetaSoundSource /MidMatchWeaponBunkerGameplay/Sound/MSS_S29_Bunker_Priming_UndergroundMech
inputs [
 InitialUnlockingSoundUnderground {type=WaveAsset}
 Pre Opening Tell {type=WaveAsset}
 Priming Duration {type=Float}
 UE.Attenuation.Distance {type=Float}
 UE.Source.OnPlay {type=Trigger}
]
interfaces [UE.Attenuation:1.0, UE.OutputFormat.Stereo:1.0, UE.Source.OneShot:1.0, UE.Source:1.1]
outputs [
 "UE.OutputFormat.Stereo.Audio:0" {type=Audio}
 "UE.OutputFormat.Stereo.Audio:1" {type=Audio}
 UE.Source.OneShot.OnFinished {type=Trigger}
]
pages [Default {
  input_defaults {
   InitialUnlockingSoundUnderground=/MidMatchWeaponBunkerGameplay/Sound/wav/S29_MMO_Bunker_Door_Unlocking
   Pre Opening Tell=/MidMatchWeaponBunkerGameplay/Sound/wav/S29_MMO_Bunker_Door_PreOpening_Open_Tell
   Priming Duration=30
   UE.Attenuation.Distance=0
   UE.Source.OnPlay=null
  }
  nodes [
   n1 = /FortMetasound/Patches/MSP_DistanceMapper(Distance=^UE.Attenuation.Distance) {also {defaults {Falloff=3000, Inner Radius=850, Inner Volume=1}}}
   n2 = MapRange.MapRange:Float(In=^UE.Attenuation.Distance) {also {defaults {In Range B=2500, Out Range A=20000, Out Range B=950}}}
   TriggerDelay_1 = UE.Trigger Delay(In=^UE.Source.OnPlay)
   WavePlayer_0 = UE.Wave Player:Stereo(Wave Asset=^InitialUnlockingSoundUnderground, Play=@TriggerDelay_1.Out)
   OnePoleLowPassFilter_1 = UE.One-Pole Low Pass Filter:Audio(Cutoff Frequency=@n2.Out Value, In=@WavePlayer_0.Out Right)
   TriggerDelay_0 = UE.Trigger Delay(
    In=^UE.Source.OnPlay
    Delay Time=UE.Subtract:Time(PrimaryOperand=UE.ConversionFloatToTime(In=^Priming Duration).Out) {also {defaults {AdditionalOperands=5}, version [1,1]}}.Out
   )
   WavePlayer_1 = UE.Wave Player:Stereo(Wave Asset=^Pre Opening Tell, Play=@TriggerDelay_0.Out)
   OnePoleLowPassFilter_0 = UE.One-Pole Low Pass Filter:Audio(Cutoff Frequency=@n2.Out Value, In=@WavePlayer_0.Out Left)
   AudioMixerAudioMixerStereo2_0 = "AudioMixer.Audio Mixer (Stereo, 2)"(
    In 0 R=UE.Multiply:Audio by Float(AdditionalOperands=@n1.Output, PrimaryOperand=@OnePoleLowPassFilter_1.Out).Out
    In 1 L=UE.Multiply:Audio by Float(AdditionalOperands=@n1.Output, PrimaryOperand=@WavePlayer_1.Out Left).Out
    In 1 R=UE.Multiply:Audio by Float(AdditionalOperands=@n1.Output, PrimaryOperand=@WavePlayer_1.Out Right).Out
    In 0 L=UE.Multiply:Audio by Float(AdditionalOperands=@n1.Output, PrimaryOperand=@OnePoleLowPassFilter_0.Out).Out
   ) {also {defaults {Gain 1=0.9}}}
   TriggerAccumulatorTriggerAccumulate2_0 = "TriggerAccumulator.Trigger Accumulate (2)"(In 0=@WavePlayer_0.On Finished, In 1=@WavePlayer_1.On Finished)
  ]
  links [
   {dst="@UE.OutputFormat.Stereo.Audio:0", src=AudioMixerAudioMixerStereo2_0.Out L}
   {dst="@UE.OutputFormat.Stereo.Audio:1", src=AudioMixerAudioMixerStereo2_0.Out R}
   {dst="@UE.Source.OneShot.OnFinished", src=TriggerAccumulatorTriggerAccumulate2_0.Out}
  ]
 }]
properties {
 assetImportData=&self:AssetImportData
 attenuationSettings=&/Att_S29_Bunker_Underground
 maxDistance=7000
 numChannels=2
 outputFormat=Stereo
 sampleRate=48000
 soundClassObject=&Fort_Ambient_Props_Athena
 volume=0.652
}
