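# One-shot mono MetaSound source for the weapon bunker purchase UI pop-out sound, using interact attenuation and the critical in-game UI sound class.
# On play, it randomizes volume and pitch, then picks a wave via a no-repeat patch and shared-state random array get, though only one wave is listed.
# The random values scale base inputs to drive gain and semitone pitch shift.
# Concurrency allows one instance and blocks new ones.
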MetaSoundSource /MidMatchWeaponBunkerGameplay/Sound/MSS_WeaponBunker_Purchase_UI_PopOut
inputs [
 MaxPitch {type=Float}
 MaxVolume {type=Float}
 MinPitch {type=Float}
 MinVolume {type=Float}
 Pitch {type=Float}
 Seed {type=Int32}
 Shared State Behavior {type=Enum:SharedStateBehaviorType}
 Sounds {type=WaveAsset:Array}
 UE.Source.OnPlay {type=Trigger}
 Volume {type=Float}
]
interfaces [UE.OutputFormat.Mono:1.0, UE.Source.OneShot:1.0, UE.Source:1.1]
outputs [
 "UE.OutputFormat.Mono.Audio:0" {type=Audio}
 UE.Source.OneShot.OnFinished {type=Trigger}
 WaveDuration {type=Time}
]
pages [Default {
  input_defaults {
   MaxPitch=1.05
   MaxVolume=1.05
   MinPitch=0.95
   MinVolume=0.95
   Pitch=1
   Seed=-1
   Shared State Behavior=1
   UE.Source.OnPlay=null
   Volume=1
   Sounds [/MidMatchWeaponBunkerGameplay/Sound/wav/WeaponBunker_Purchase_UI_PopOut_01]
  }
  nodes [
   n1 = UE.RandomFloat(Max=^MaxVolume, Min=^MinVolume, Seed=^Seed, Next=^UE.Source.OnPlay)
   n2 = UE.RandomFloat(Max=^MaxPitch, Min=^MinPitch, Seed=^Seed, Next=@n1.On Next)
   n3 = /FortMetasound/Patches/MSP_GetNoRepeatNumber(WaveArray=^Sounds, InTrigger=@n2.On Next)
   ArrayRandomGet_0 = Array.Random Get:WaveAsset:Array(
    Seed=^Seed
    Shared State Behavior=^Shared State Behavior
    In Array=^Sounds
    Next=@n3.OutTrigger
    No Repeats=UE.Reroute(Value=@n3.NoRepeatNumber).Value
   ) {also {defaults {Enable Shared State="<MetaSoundBuilderResult.SUCCEEDED: 0>"}}}
   GetWaveDuration_0 = "UE.Get Wave Duration: "(Wave=@ArrayRandomGet_0.Value)
   FrequencyMultipliertoSemitone_0 = UE.Frequency Multiplier to Semitone:Float(Frequency Multiplier=UE.Reroute(Value=UE.Multiply:Float(AdditionalOperands=^Pitch, PrimaryOperand=@n2.Value).Out).Value)
   WavePlayer_0 = UE.Wave Player:Mono(Play=@ArrayRandomGet_0.On Next, Wave Asset=@ArrayRandomGet_0.Value, Pitch Shift=@FrequencyMultipliertoSemitone_0.Semitones) {also {version [1,1]}}
   Multiply_0 = UE.Multiply:Audio by Float(
    AdditionalOperands=UE.Reroute(Value=UE.Multiply:Float(AdditionalOperands=^Volume, PrimaryOperand=@n1.Value).Out).Value
    PrimaryOperand=@WavePlayer_0.Out Mono
   )
  ]
  links [
   {dst="@WaveDuration", src=GetWaveDuration_0.Duration}
   {dst="@UE.OutputFormat.Mono.Audio:0", src=Multiply_0.Out}
   {dst="@UE.Source.OneShot.OnFinished", src=WavePlayer_0.On Finished}
  ]
 }]
properties {
 assetImportData=&self:AssetImportData
 attenuationSettings=/Game/Sounds/Attenuation/Fort_Item_Interact_Mono_Att
 bOverrideConcurrency=true
 maxDistance=1500
 sampleRate=48000
 soundClassObject=&Fort_UI_InGame_Critical
 volume=0.2
 concurrencyOverrides {maxCount=1, resolutionRule=PreventNew}
}
