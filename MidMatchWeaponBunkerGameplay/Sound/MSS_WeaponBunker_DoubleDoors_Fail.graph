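# Mono one-shot MetaSound source for the weapon bunker double-door fail sound, playing a military-rank scan-fail wave.
# It feeds OnPlay and the Locked sound, volume and pitch-range inputs into a reused Purchase_UI_PopOut MetaSound, which supplies audio and OnFinished.
# The Lock.Status and Unlock inputs are declared but unconnected.
# It uses door-slam attenuation and the player-interacts sound class.
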MetaSoundSource /MidMatchWeaponBunkerGameplay/Sound/MSS_WeaponBunker_DoubleDoors_Fail
inputs [
 Lock.Status {type=Bool}
 Locked.MasterVolume {type=Float}
 Locked.MaxPitch {type=Float}
 Locked.MaxVolume {type=Float}
 Locked.MinPitch {type=Float}
 Locked.MinVolume {type=Float}
 Locked.Pitch {type=Float}
 Locked.Sounds {type=WaveAsset:Array}
 UE.Source.OnPlay {type=Trigger}
 Unlock.Master Volume {type=Float}
 Unlock.MaxPitch {type=Float}
 Unlock.MaxVolume {type=Float}
 Unlock.MinPitch {type=Float}
 Unlock.MinVolume {type=Float}
 Unlock.Pitch {type=Float}
 Unlock.Sounds {type=WaveAsset:Array}
]
interfaces [UE.OutputFormat.Mono:1.0, UE.Source.OneShot:1.0, UE.Source:1.1]
outputs [
 "UE.OutputFormat.Mono.Audio:0" {type=Audio}
 UE.Source.OneShot.OnFinished {type=Trigger}
]
pages [Default {
  input_defaults {
   Lock.Status=null
   Locked.MasterVolume=1
   Locked.MaxPitch=1.05
   Locked.MaxVolume=1.05
   Locked.MinPitch=0.95
   Locked.MinVolume=0.95
   Locked.Pitch=1
   UE.Source.OnPlay=null
   Unlock.Master Volume=1
   Unlock.MaxPitch=1.05
   Unlock.MaxVolume=1.05
   Unlock.MinPitch=0.95
   Unlock.MinVolume=0.95
   Unlock.Pitch=1
   Locked.Sounds [/MidMatchWeaponBunkerGameplay/Sound/wav/WeaponBunker_MilitaryRank_Scan_Fail_01]
   Unlock.Sounds []
  }
  nodes [MSS_WeaponBunker_Purchase_UI_PopOut_0 = &/MSS_WeaponBunker_Purchase_UI_PopOut(
    Volume=^Locked.MasterVolume
    MaxPitch=^Locked.MaxPitch
    MaxVolume=^Locked.MaxVolume
    MinPitch=^Locked.MinPitch
    MinVolume=^Locked.MinVolume
    Pitch=^Locked.Pitch
    Sounds=^Locked.Sounds
    UE.Source.OnPlay=^UE.Source.OnPlay
   )]
  links [
   {dst="@UE.OutputFormat.Mono.Audio:0", src=MSS_WeaponBunker_Purchase_UI_PopOut_0.UE.OutputFormat.Mono.Audio:0}
   {dst="@UE.Source.OneShot.OnFinished", src=MSS_WeaponBunker_Purchase_UI_PopOut_0.UE.Source.OneShot.OnFinished}
  ]
 }]
properties {
 assetImportData=&self:AssetImportData
 attenuationSettings=/Game/Sounds/Attenuation/Fort_Doors_Slam_Open_Att
 maxDistance=2500
 sampleRate=48000
 soundClassObject=/Game/Sounds/FortSoundClassesAndModes/Fort_08_Player_Interacts
 volume=1.5
}
