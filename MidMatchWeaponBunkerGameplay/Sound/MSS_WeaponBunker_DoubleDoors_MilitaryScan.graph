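# Mono one-shot MetaSound for the military rank scan at Weapon Bunker double doors, playing a scan loop wave.
# Scan.Duration and Ratio derive equal attack and release times plus a delayed release trigger, fed back through a deferred variable.
# One ADSR envelope bends pitch; another shapes volume with a doubled release and fires OnFinished.
# Playback starts on source OnPlay; Play, Lock.Status and other Scan inputs are unconnected.
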MetaSoundSource /MidMatchWeaponBunkerGameplay/Sound/MSS_WeaponBunker_DoubleDoors_MilitaryScan
inputs [
 Lock.Status {type=Bool}
 Play {type=Trigger}
 Ratio {type=Float}
 Scan.CheckStatus {type=Trigger}
 Scan.Duration {type=Float}
 Scan.LockTest {type=Trigger}
 Scan.Start {type=Trigger}
 UE.Source.OnPlay {type=Trigger}
]
interfaces [UE.OutputFormat.Mono:1.0, UE.Source.OneShot:1.0, UE.Source:1.1]
outputs [
 "UE.OutputFormat.Mono.Audio:0" {type=Audio}
 UE.Source.OneShot.OnFinished {type=Trigger}
]
pages [Default {
  input_defaults {
   Lock.Status=null
   Play=null
   Ratio=1.5
   Scan.CheckStatus=null
   Scan.Duration=1.64
   Scan.LockTest=null
   Scan.Start=null
   UE.Source.OnPlay=null
  }
  nodes [
   InitVariable.Float()
   InitVariable.WaveAsset:Array()
   n1 = UE.Divide:Float(PrimaryOperand=^Scan.Duration, AdditionalOperands=UE.Add:Float(AdditionalOperands=^Ratio) {also {defaults {PrimaryOperand=2}}}.Out)
   n2 = VariableAccessor.Time(Variable=VariableMutator.Time(Value=UE.ConversionFloatToTime(In=@n1.Out).Out, Variable=InitVariable.Time().Variable).Variable)
   n3 = VariableAccessor.Time(Variable=@n2.Variable)
   n4 = VariableAccessor.Time(Variable=@n3.Variable)
   n5 = VariableDeferredAccessor.Trigger(Variable=InitVariable.Trigger().Variable)
   n6 = VariableDeferredAccessor.Trigger(Variable=@n5.Variable)
   ADSREnvelopeADSREnvelope_0 = ADSR Envelope.ADSR Envelope:Float(Trigger Attack=^UE.Source.OnPlay, Release Time=@n3.Value, Attack Time=@n2.Value, Trigger Release=@n6.Value) {also {defaults {Decay Time=0, Sustain Level=1.28}}}
   WavePlayer_0 = UE.Wave Player:Mono(
    Play=@ADSREnvelopeADSREnvelope_0.On Attack Triggered
    Pitch Shift=MapRange.MapRange:Float(In=@ADSREnvelopeADSREnvelope_0.Out Envelope) {also {defaults {Out Range B=3}}}.Out Value
   ) {also {
     defaults {Loop="<MetaSoundBuilderResult.SUCCEEDED: 0>", Wave Asset=/MidMatchWeaponBunkerGameplay/Sound/wav/WeaponBunker_MilitaryRank_Scan_Loop_01}
     version [1,1]
    }}
   ADSREnvelopeADSREnvelope_1 = ADSR Envelope.ADSR Envelope:Float(
    Release Time=UE.Multiply:Time by Float(PrimaryOperand=@n4.Value) {also {defaults {AdditionalOperands=2}}}.Out
    Trigger Attack=UE.Reroute(Value=@WavePlayer_0.On Play).Value
    Attack Time=VariableAccessor.Time(Variable=@n4.Variable).Value
    Trigger Release=@n5.Value
   ) {also {defaults {Sustain Level=1}}}
   Multiply_0 = UE.Multiply:Audio by Float(AdditionalOperands=@ADSREnvelopeADSREnvelope_1.Out Envelope, PrimaryOperand=@WavePlayer_0.Out Mono)
   n7 = UE.ConversionFloatToTime(In=UE.Multiply:Float(PrimaryOperand=^Ratio, AdditionalOperands=@n1.Out).Out)
   n8 = InitVariable.Time()
   TriggerDelay_0 = UE.Trigger Delay(
    In=@ADSREnvelopeADSREnvelope_0.On Decay Triggered
    Delay Time=VariableAccessor.Time(Variable=VariableMutator.Time(Value=@n7.Out, Variable=@n8.Variable).Variable).Value
   )
   VariableMutator.Trigger(Value=@TriggerDelay_0.Out, Variable=@n6.Variable)
  ]
  links [{dst="@UE.Source.OneShot.OnFinished", src=ADSREnvelopeADSREnvelope_1.On Done}, {dst="@UE.OutputFormat.Mono.Audio:0", src=Multiply_0.Out}]
 }]
properties {
 assetImportData=&self:AssetImportData
 attenuationSettings=/Game/Sounds/Attenuation/Fort_Doors_Less_Reverb_Att
 maxDistance=2400
 sampleRate=48000
 soundClassObject=/Game/Sounds/FortSoundClassesAndModes/Fort_08_Player_Interacts
 concurrencySet [&/CONC_WeaponBunker_MilitaryRankScan]
}
