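# MetaSound Source preset that plays the start sound of the military rank scan at the Weapon Bunker double doors.
# It inherits the mono, one-shot random-sound parent graph and supplies a single scan-start wave with slight pitch and volume ranges.
# Playback uses reduced-reverb door attenuation, the player-interaction sound class and a dedicated military-rank-scan concurrency set.
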MetaSoundSource /MidMatchWeaponBunkerGameplay/Sound/MSS_WeaponBunker_DoubleDoors_MilitaryScan_Start
preset_parent=&MSS_PlayRandomSound_Mono
inputs [
 MaxPitch {type=Float}
 MaxVolume {type=Float}
 MinPitch {type=Float}
 MinVolume {type=Float}
 Pitch {type=Float}
 Seed {type=Int32}
 Shared State Behavior {type=Enum:SharedStateBehaviorType}
 Sounds {type=WaveAsset:Array}
 UE.Source.OnPlay {type=Trigger}
 Volume {type=Float}
]
interfaces [UE.OutputFormat.Mono:1.0, UE.Source.OneShot:1.0, UE.Source:1.1]
outputs [
 "UE.OutputFormat.Mono.Audio:0" {type=Audio}
 UE.Source.OneShot.OnFinished {type=Trigger}
 WaveDuration {type=Time}
]
pages [Default {input_defaults {
   MaxPitch=1
   MaxVolume=1.05
   MinPitch=0.95
   MinVolume=0.95
   Pitch=1
   Seed=-1
   Shared State Behavior=1
   UE.Source.OnPlay=null
   Volume=1
   Sounds [/MidMatchWeaponBunkerGameplay/Sound/wav/WeaponBunker_MilitaryRank_Scan_Start_01]
  }}]
properties {
 assetImportData=&self:AssetImportData
 attenuationSettings=/Game/Sounds/Attenuation/Fort_Doors_Less_Reverb_Att
 maxDistance=2400
 sampleRate=48000
 soundClassObject=/Game/Sounds/FortSoundClassesAndModes/Fort_08_Player_Interacts
 volume=0.8
 concurrencySet [&/CONC_WeaponBunker_MilitaryRankScan]
}
